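# Physics asset for the whip prop used in the Black Monday female whip emote, simulating the small dangling loop on the whip's handle.
# A long capsule covers the handle, and three short capsules cover the dyn_loop bone chain.
# Angular-limited constraints link the handle to the first loop bone and each loop bone to the next, with wider swing at the handle joint and tight twist throughout.
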PhysicsAsset /Game/Animation/Props/BlackMonday_Female_Whip/Meshes/BlackMondayFemale_Emote_Whip_Physics
bodies {
 dyn_loop_01 [{shapeType=Capsule, length=3, radius=1.5, center=<2 0 0>, rotation=<rot 90 0 0>}]
 dyn_loop_02 [{shapeType=Capsule, length=3, radius=1.5, center=<2 0 0>, rotation=<rot 90 0 0>}]
 dyn_loop_03 [{shapeType=Capsule, length=3, radius=1.5, center=<2 0 0>, rotation=<rot 90 0 0>}]
 handle [{shapeType=Capsule, length=25.0439, radius=2.1188, rotation=<rot 90 0 0>}]
}
constraints [
 {bone1Name=dyn_loop_01, bone2Name=handle, swing1LimitDegrees=60, swing2LimitDegrees=60, twistLimitDegrees=15}
 {bone1Name=dyn_loop_02, bone2Name=dyn_loop_01, swing1LimitDegrees=45, swing2LimitDegrees=45, twistLimitDegrees=15}
 {bone1Name=dyn_loop_03, bone2Name=dyn_loop_02, swing1LimitDegrees=45, swing2LimitDegrees=45, twistLimitDegrees=15}
]
